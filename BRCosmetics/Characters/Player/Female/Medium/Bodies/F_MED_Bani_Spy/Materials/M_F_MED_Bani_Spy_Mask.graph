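# Skeletal character material for the F_MED_Bani_Spy mask, adding a glitching, static-distorted emissive to the shared character PreFX/PostFX material.
# Panned static and a scrolling sweep line distort the emissive mask UVs and modulate brightness, periodically switching to stronger "big glitch" values.
# A smoothstep on pre-skinned local height limits the distortion.
# The Fresnel-weighted emissive blends in through a SkinFX mask channel, then passes through MF_EmissiveFresnel before PostFX.
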Material /BRCosmetics/Characters/Player/Female/Medium/Bodies/F_MED_Bani_Spy/Materials/M_F_MED_Bani_Spy_Mask
graph {
 properties {
  bUseMaterialAttributes=true
  bUsedWithMeshDeformer=true
  bUsedWithMorphTargets=true
  bUsedWithSkeletalMesh=true
  pixelDepthOffsetMode=PDOM_Legacy
  shadingModel=MSM_FromMaterialExpression
  subsurfaceProfile=&SS_HeroSkin_02
  usedShadingModels=Default Lit | Subsurface Profile
 }
 nodes [
  n1 = MaterialFunctionCall() {materialFunction=&BaseFN_CharacterFunction_PreFX}
  n2 = Saturate(None=Fresnel(ExponentIn=ScalarParameter() {defaultValue=4, group=Static, parameterName=DistortedEmissive_Fresnel Exponent}) {baseReflectFraction=0})
  n3 = LinearInterpolate(A=OneMinus(None=@n2), B=@n2, Alpha=ScalarParameter() {group=Static, parameterName=DistortedEmissive_Invert Fresnel})
  n4 = Multiply(A=ComponentMask(None=TextureCoordinate()) {g=true}) {constB=0.75}
  n5 = Multiply(A=Time()) {constB=-0.2}
  n6 = Power(Base=Frac(None=Add(A=@n4, B=@n5))) {constExponent=20}
  n7 = Multiply(A=Time(), B=ScalarParameter() {defaultValue=0.08, parameterName=BigGlitchSpeed})
  n8 = Round(None=Power(Base=Frac(None=@n7)))
  n9 = LinearInterpolate(
   A=ScalarParameter() {defaultValue=0.01, group=Sweep Line, parameterName=Sweep UV Distort}
   B=ScalarParameter() {defaultValue=0.01, group=Sweep Line, parameterName=Sweep UV Big Distort}
   Alpha=@n8
  )
  n10 = TextureCoordinate()
  n11 = LinearInterpolate(A=@n10, B=Multiply(A=@n10) {constB=0.75}, Alpha=@n8)
  n12 = Multiply(
   A=TextureSample(UVs=Panner(Coordinate=@n11) {speedX=6, speedY=-0.05}) {samplerType=SAMPLERTYPE_Masks, texture=/Game/Missions/Primary/RepairTheShelter/Textures/Cinematic_Screen_Static}.G
   B=TextureSample(UVs=Panner(Coordinate=@n11) {speedX=-3, speedY=-0.03}) {samplerType=SAMPLERTYPE_Masks, texture=&Cinematic_Screen_Static}.R
  )
  n13 = LinearInterpolate(
   A=ScalarParameter() {defaultValue=0.07, group=Static, parameterName=Static UV Distort}
   B=ScalarParameter() {defaultValue=5, group=Static, parameterName=Static UV Big Distort}
   Alpha=@n8
  )
  n14 = Add(A=AppendVector(A=Multiply(A=@n6, B=@n9) {constB=0.02}, B=Constant()), B=AppendVector(A=Multiply(A=@n12, B=@n13), B=Constant()))
  n15 = PreSkinnedPosition()
  n16 = PreSkinnedLocalBounds()
  n17 = VertexInterpolator(VS=Divide(A=Subtract(A=@n15, B=@n16.Min), B=@n16.Extents))
  n18 = Saturate(None=MaterialFunctionCall(
    Alpha=ComponentMask(None=@n17.PS) {b=true}
    Min=ScalarParameter() {defaultValue=0.15, parameterName=AttenLow}
    Max=ScalarParameter() {defaultValue=0.2, parameterName=AttenHigh}
   ) {materialFunction=&SmoothStep}.Result)
  n19 = TextureSample(UVs=Add(A=TextureCoordinate(), B=Multiply(A=@n14, B=@n18))) {
   samplerType=SAMPLERTYPE_Masks
   texture=/BRCosmetics/Characters/Player/Female/Medium/Bodies/F_MED_Bani_Spy/Textures/T_F_MED_Bani_Spy_FaceAcc_Mask_E
  }
  n20 = Multiply(
   A=Power(Base=@n19.RGB, Exp=ScalarParameter() {defaultValue=1, group=Static, parameterName=DistortedEmissive_BrightnessPow})
   B=ScalarParameter() {defaultValue=70, group=Static, parameterName=DistortedEmissive_BrightnessMax}
  ) {constB=70}
  n21 = Panner(Coordinate=@n11) {speedX=5, speedY=0.1}
  n22 = Add(
   A=Multiply(A=TextureSample(UVs=@n21) {samplerType=SAMPLERTYPE_Masks, texture=&Cinematic_Screen_Static}.B) {constB=0.1}
   B=@n12
  )
  n23 = LinearInterpolate(
   A=ScalarParameter() {defaultValue=3, group=Static, parameterName=Static Scale}
   B=ScalarParameter() {defaultValue=10, group=Static, parameterName=Static Big Scale}
   Alpha=@n8
  )
  n24 = LinearInterpolate(
   A=ScalarParameter() {group=Static, parameterName=StaticLowValue}
   B=ScalarParameter() {defaultValue=1, group=Static, parameterName=StaticHighValue}
   Alpha=Multiply(A=Multiply(A=@n22, B=@n23), B=@n18)
  )
  n25 = Multiply(A=Multiply(A=@n6, B=ScalarParameter() {defaultValue=0.3, group=Sweep Line, parameterName=Sweep Add}) {constB=0.2}, B=@n18)
  n26 = SetMaterialAttributes!(
   MaterialAttributes=@n1.Result
   Emissive Color=Multiply(A=Multiply(A=@n3, B=@n20), B=Add(A=@n24, B=@n25))
   Specular=Multiply(
    A=GetMaterialAttributes(None=@n1.Result) {attributeGetTypes [#1]}.Specular
    B=ScalarParameter() {defaultValue=1, group=Static, parameterName=SpecularMultiplier}
   )
  ) {attributeSetTypes [#2, #1]}
  n27 = StaticComponentMaskParameter(None=TextureSample(Tex=TextureObjectParameter() {group=Skin FX, parameterName=SkinFX_Mask}).RGB) {defaultR=true, group=Skin FX, parameterName=DistortedEmissive_Channel}
  MaterialFunctionCall(
   AO=@n1.AO
   Distance Blend=@n1.AO
   Fresnel=@n1.AO
   Material=MaterialFunctionCall!(In=BlendMaterialAttributes(A=@n1.Result, B=@n26, Alpha=@n27)) {materialFunction=/Game/Weapons/WeaponSkins/MaterialFunction/MF_EmissiveFresnel}.Result
  ) {materialFunction=&BaseFN_CharacterFunction_PostFX}
 ]
}
